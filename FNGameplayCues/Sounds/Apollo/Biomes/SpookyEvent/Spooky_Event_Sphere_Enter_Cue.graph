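# Sound cue for entering the spooky event sphere, an outdoor ambient biome one-shot in the Apollo map.
# A single wave player plays the Spooky_Event_Enter wave through a modulator that adds very slight random pitch variation.
# It reuses the Hightower Sphere enter/exit attenuation, plays in the Athena outdoor ambient sound class, and has reduced volume.
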SoundCue /FNGameplayCues/Sounds/Apollo/Biomes/SpookyEvent/Spooky_Event_Sphere_Enter_Cue
root=Modulator_0
nodes [
 {id=Modulator_0, children [WavePlayer_0], properties {pitchMax=1.01, pitchMin=0.99}}
 {id=WavePlayer_0, properties {soundWaveAssetPtr=&/Spooky_Event_Enter}}
]
properties {
 attenuationSettings=/FNGameplayCues/Sounds/Apollo/Biomes/Hightower_Sphere/Hightower_Sphere_EnterExit_Att
 maxDistance=5000
 soundClassObject=&Fort_Ambient_Outdoors_Athena
 volumeMultiplier=0.626
 preEffectBusSends [{}]
}
